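# Translucent UI material that draws animated glowing, color-shifting ribbons on a virtual quad in perspective.
# A Custom HLSL ray-plane intersection turns center-scaled UVs into quad UVs from Translation and Rotation parameters, flagging outside pixels with -1.
# A second Custom node loops over sine-displaced lines to build the glow (its Ribbons_Glow input is unused).
# The result fades horizontally, is multiplied by a quad-bounds term and can be blended with an inverted sphere mask.
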Material /FortniteCoreUI/UI/Foundation/RockyRidge/Materials/M_UI_NightNight_Ribbons
graph {
 properties {
  bUsedWithStaticMesh=false
  blendMode=BLEND_Translucent
  materialDomain=MD_UI
  pixelDepthOffsetMode=PDOM_Legacy
  usedShadingModels=Default Lit
 }
 nodes [
  n1 = VectorParameter() {parameterName=UVScale, defaultValue=<rgba 1 1 0 1>}
  MaterialExpressionMaterialFunctionCall_1 = MaterialFunctionCall(UVs=TextureCoordinate(), Texture Scale=AppendVector(A=@n1.R, B=@n1.R)) {materialFunction=&ScaleUVsByCenter}
  MaterialExpressionMaterialFunctionCall_0 = MaterialFunctionCall() {materialFunction=&GetUserInterfaceUV}
  MaterialExpressionComponentMask_0 = ComponentMask(None=VectorParameter() {parameterName=Translation, defaultValue=<rgba 0 0 -0.8002 1>}.RGB) {b=true, g=true, r=true}
  MaterialExpressionComponentMask_1 = ComponentMask(None=VectorParameter() {parameterName=Rotation, defaultValue=<rgba 0 0.4286 0 1>}.RGB) {b=true, g=true, r=true}
  n2 = Custom!(
   InUV=@MaterialExpressionMaterialFunctionCall_1.UVs
   InResolution=@MaterialExpressionMaterialFunctionCall_0.Pixel Size
   InTranslation=@MaterialExpressionComponentMask_0
   InRotation=@MaterialExpressionComponentMask_1
  ) {
   code="float2 scUV = InUV;\r\nscUV.y = 1. - scUV.y;\r\n\r\nfloat2 p = (2*scUV)-1.;\r\nfloat scAspect = InResolution.x/InResolution.y;\r\n\r\nfloat3 dir = float3(p.x, p.y, 1);\r\ndir = dir/length(dir);\r\n\r\nfloat3 pPos = InTranslation;\r\nfloat3 pRot = InRotation;\r\nfloat2 pDimension = float2(-scAspect, 1.);\r\n\r\nfloat2 uv = GetQuadUV(float3(0.,0.,0.), dir, pPos, pRot, pDimension);\r\n\r\nif(abs(uv.x-.5) < .5 &&&& abs(uv.y - .5) < .5)\r\n{\r\n    return uv;\r\n}\r\nelse\r\n{\r\n    return float2(-1.,-1.);\r\n}"
   outputType=CMOT_Float2
   preamble="MaterialFloat GetPlaneIntersectionDistance(\r\n    MaterialFloat3 InNor,\r\n    MaterialFloat3 InPos,\r\n    MaterialFloat3 InRo,\r\n    MaterialFloat3 InRd\r\n)\r\n{\r\n    float de = dot(InNor, InRd);\r\n    de = sign(de) * max(abs(de), 0.001);\r\n    return dot(InNor, InPos - InRo) / de;\r\n}\r\n\r\nMaterialFloat2 GetQuadUV(\r\n    MaterialFloat3 InRO,\r\n    MaterialFloat3 InRD,\r\n    MaterialFloat3 InQCen,\r\n    MaterialFloat3 InQRot,\r\n    MaterialFloat2 InQDim\r\n)\r\n{\r\n    float a = sin(InQRot.x);\r\n    float b = cos(InQRot.x);\r\n    float c = sin(InQRot.y);\r\n    float d = cos(InQRot.y);\r\n    float e = sin(InQRot.z);\r\n    float f = cos(InQRot.z);\r\n\r\n    float ac = a*c;\r\n    float bc = b*c;\r\n\r\n    float3x3 rotMat = float3x3( float3(d*f, d*e, -c),\r\n                                float3(ac*f-b*e, ac*e+b*f, a*d),\r\n                                float3(bc*f+a*e, bc*e-a*f, b*d));\r\n\r\n    float3 rt = mul(rotMat, float3(InQDim.x, .0, .0)).xyz;\r\n    float3 up = mul(rotMat, float3(0, InQDim.y, 0 )).xyz;\r\n    float3 normal = cross(rt, up);\r\n    normal = normal/length(normal);\r\n\r\n    float3 pos = (InRD * GetPlaneIntersectionDistance(normal, InQCen, InRO, InRD)) - InQCen;\r\n\r\n    return float2(dot(pos, rt) / dot(rt, rt), dot(pos, up) / dot(up, up)) + .5;\r\n}\r\n"
   inputs [
    {
     inputName=InUV
     input {expression=&self:MaterialExpressionMaterialFunctionCall_1}
    }
    {
     inputName=InResolution
     input {
      expression=&self:MaterialExpressionMaterialFunctionCall_0
      outputIndex=1
     }
    }
    {
     inputName=InTranslation
     input {expression=&self:MaterialExpressionComponentMask_0}
    }
    {
     inputName=InRotation
     input {expression=&self:MaterialExpressionComponentMask_1}
    }
   ]
  }
  n3 = VectorParameter() {parameterName=SphereMask, defaultValue=<rgba 0.5 0.5 0.5 0.7524>}
  n4 = AppendVector(A=@n3.R, B=@n3.R)
  n5 = Saturate(None=OneMinus(None=SphereMask(A=@n2, B=@n4, Radius=@n3.B, Hardness=@n3.B)))
  n6 = ScalarParameter() {defaultValue=1, parameterName=SphereMask_Scalar}
  n7 = Add(A=Multiply(A=@n2)) {constB=0}
  MaterialExpressionSaturate_0 = Saturate(None=@n7)
  MaterialExpressionTime_0 = Time()
  MaterialExpressionScalarParameter_0 = ScalarParameter() {defaultValue=3, parameterName=Ribbons_Num}
  MaterialExpressionScalarParameter_1 = ScalarParameter() {defaultValue=0.01, parameterName=Ribbons_Intensity_Min}
  MaterialExpressionScalarParameter_2 = ScalarParameter() {defaultValue=150, parameterName=Ribbons_Squeeze}
  MaterialExpressionScalarParameter_3 = ScalarParameter() {defaultValue=1.6, parameterName=Ribbons_Glow}
  MaterialExpressionScalarParameter_5 = ScalarParameter() {defaultValue=0.13, parameterName=Sin_Offset_A}
  MaterialExpressionScalarParameter_6 = ScalarParameter() {defaultValue=0.23, parameterName=Sin_Offset_B}
  MaterialExpressionScalarParameter_7 = ScalarParameter() {defaultValue=0.19, parameterName=Sin_Offset_C}
  MaterialExpressionScalarParameter_8 = ScalarParameter() {defaultValue=1.6, parameterName=IntensityScalar}
  MaterialExpressionScalarParameter_9 = ScalarParameter() {defaultValue=0.6, parameterName=TimeScale_A}
  MaterialExpressionComponentMask_4 = ComponentMask(None=VectorParameter() {parameterName=ColorIntensity, defaultValue=<rgba 1.5 1 1 1>}.RGB) {b=true, g=true, r=true}
  n8 = Custom!(
   InUv=@MaterialExpressionSaturate_0
   InTime=@MaterialExpressionTime_0
   InNumLines=@MaterialExpressionScalarParameter_0
   InMinIntensity=@MaterialExpressionScalarParameter_1
   InWidthScalar=@MaterialExpressionScalarParameter_2
   InGlowScalar=@MaterialExpressionScalarParameter_3
   InOffsetA=@MaterialExpressionScalarParameter_5
   InOffsetB=@MaterialExpressionScalarParameter_6
   InOffsetC=@MaterialExpressionScalarParameter_7
   InIntense=@MaterialExpressionScalarParameter_8
   InTimeScaleA=@MaterialExpressionScalarParameter_9
   InColorIntensity=@MaterialExpressionComponentMask_4
  ) {
   code="float2 uv = InUv;\r\nuv = -1. + 2. * uv;\r\n\r\nfloat time = InTime;\r\nfloat numLines = InNumLines;\r\nfloat minIntens = InMinIntensity;\r\nfloat widthScalar = InWidthScalar;\r\nfloat3 colorInt = InColorIntensity;\r\nfloat intensity;\r\nfloat width;\r\nfloat3 c = float3(0.,0.,0.);\r\n\r\nfor(float i = 0.; i < numLines; i++)\r\n{\r\n\tuv.y += (.2* sin(uv.x + i/1. - time * InTimeScaleA));\r\n\tfloat Y = uv.y + .2 * .2;\r\n\tintensity = minIntens + pow( InIntense * abs(fmod(InUv.x + i/1.15 + time * .8, 2.) - 1.0), 2.);\r\n\twidth = abs(intensity / (widthScalar * Y));\r\n\tc += float3(width * (colorInt.x + sin(time * InOffsetA)), \r\n\t\t\t    width * (colorInt.y - sin(time * InOffsetB)),\r\n\t\t\t    width * (colorInt.z - cos(time * InOffsetC)));\r\n\r\n}\r\nc = lerp(c, float3(0., 0., 0.), uv.x);\r\n\r\nreturn c;\r\n"
   inputs [
    {inputName=InUv, input {expression=&self:MaterialExpressionSaturate_0}}
    {inputName=InTime, input {expression=&self:MaterialExpressionTime_0}}
    {
     inputName=InNumLines
     input {expression=&self:MaterialExpressionScalarParameter_0}
    }
    {
     inputName=InMinIntensity
     input {expression=&self:MaterialExpressionScalarParameter_1}
    }
    {
     inputName=InWidthScalar
     input {expression=&self:MaterialExpressionScalarParameter_2}
    }
    {
     inputName=InGlowScalar
     input {expression=&self:MaterialExpressionScalarParameter_3}
    }
    {
     inputName=InOffsetA
     input {expression=&self:MaterialExpressionScalarParameter_5}
    }
    {
     inputName=InOffsetB
     input {expression=&self:MaterialExpressionScalarParameter_6}
    }
    {
     inputName=InOffsetC
     input {expression=&self:MaterialExpressionScalarParameter_7}
    }
    {
     inputName=InIntense
     input {expression=&self:MaterialExpressionScalarParameter_8}
    }
    {
     inputName=InTimeScaleA
     input {expression=&self:MaterialExpressionScalarParameter_9}
    }
    {
     inputName=InColorIntensity
     input {expression=&self:MaterialExpressionComponentMask_4}
    }
   ]
  }
  n9 = Constant3Vector()
  n10 = Power(
   Base=MaterialFunctionCall(Float2=@MaterialExpressionSaturate_0) {materialFunction=&BreakOutFloat2Components}.R
   Exp=ScalarParameter() {defaultValue=1, parameterName=FadeFalloff_Pow}
  )
  n11 = Multiply(A=ComponentMask(None=Ceil(None=@n7)) {r=true}, B=Saturate(None=LinearInterpolate(A=@n8, B=@n9, Alpha=@n10)))
  LinearInterpolate(A=@n11, B=Multiply(A=Multiply(A=@n5, B=@n6), B=@n11), Alpha=ScalarParameter() {parameterName=SphereMask_Enable})
 ]
}
